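# Animation Blueprint for the Meteor Man Remix large male body, correcting secondary costume bones on a pose copied from the parent mesh.
# Weighted constraints position shoulder pads, blades, feet and thigh bolts, while additive bone transforms offset shoulder pads and scale sub-thigh bones.
# The EventGraph maps arm and thigh swing from the ref pose to those transforms' alphas.
# Skydive and parachute states override shoulder pad alphas, with extra offsets while floating.
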AnimBlueprint /BRCosmetics/Characters/Player/Male/Large/Bodies/M_LRG_Meteor_Man_Remix/Meshes/M_LRG_Meteor_Man_Remix_AnimBp
target_skeleton=&/M_LRG_Meteor_Man_Remix_Skeleton
graphs [
 AnimGraph {nodes [
   n1 = "Constraint - dyn_shoulderpad_r (PPPPPPPPP)"!(
    ComponentPose=Local To Component(LocalPose=Copy Pose From Mesh() {bUseAttachedParent=true})
    Alpha=1
    ConstraintWeights_0=0.0002
    ConstraintWeights_1:float=0.1119
    ConstraintWeights_2:float=0.0004
    ConstraintWeights_3:float=0.3367
    ConstraintWeights_4:float=0.2039
    ConstraintWeights_5:float=0.0296
    ConstraintWeights_6:float=0.0011
    ConstraintWeights_7:float=0.0035
    ConstraintWeights_8:float=0.3127
   ) -> (Pose) {
    alphaBoolBlend {blendOption=Linear}
    boneToModify {boneName=dyn_shoulderpad_r}
    constraintSetup [
     {targetBone {boneName=lowerarm_twist_02_r}}
     {targetBone {boneName=upperarm_r}}
     {targetBone {boneName=lowerarm_r}}
     {targetBone {boneName=upperarm_twist_01_r}}
     {targetBone {boneName=upperarm_twist_02_r}}
     {targetBone {boneName=spine_03}}
     {targetBone {boneName=spine_04}}
     {targetBone {boneName=spine_05}}
     {targetBone {boneName=clavicle_r}}
    ]
    constraintWeights [1,1,1,1,1,1,1,1,1]
   }
   n2 = Get bIsParachuteOpened()
   n3 = Get ParachuteOpened() -> (ParachuteOpened:double)
   n4 = Get IsSkydiving() -> (IsSkydiving:double)
   n5 = "Transform (Modify) Bone - Bone: dyn_shoulderpad_r"(
    ComponentPose=@n1
    Alpha=SelectFloat(A=SelectFloat(A=@n3, B=@n4, bPickA=@n2), B=Get ArmSwingDelta_R() -> (ArmSwingDelta_R:double), bPickA=Get FloatOrDive() -> (FloatOrDive:bool))
    Translation="0,0,0.3"
    Rotation="5,0,20"
    Scale="1,1,1"
   ) {
    rotationMode=BMM_Additive
    rotationSpace=BCS_BoneSpace
    translationMode=BMM_Additive
    translationSpace=BCS_BoneSpace
    alphaBoolBlend {blendOption=Linear}
    boneToModify {boneName=dyn_shoulderpad_r}
    rotation {pitch=5, roll=20}
    translation {z=0.3}
   }
   n6 = "Transform (Modify) Bone - Bone: dyn_shoulderpad_r"!(ComponentPose=@n5, bAlphaBoolEnabled:bool=Get bIsSkydiveFloating(), Translation="0,0,0.5", Rotation="5,10,-15", Scale="1,1,1") -> (Pose) {
    alphaInputType=Bool
    rotationMode=BMM_Additive
    rotationSpace=BCS_BoneSpace
    translationMode=BMM_Additive
    translationSpace=BCS_BoneSpace
    alphaBoolBlend {blendOption=Linear}
    boneToModify {boneName=dyn_shoulderpad_r}
    rotation=<rot 5 10 -15>
    translation {z=0.5}
   }
   n7 = "Constraint - dyn_shoulderpad_l (PPPPP)"!(
    ComponentPose=@n6
    Alpha=1
    ConstraintWeights_0=0.2062
    ConstraintWeights_1:float=0.0352
    ConstraintWeights_2:float=0.1072
    ConstraintWeights_3:float=0.308
    ConstraintWeights_4:float=0.3434
   ) -> (Pose) {
    alphaBoolBlend {blendOption=Linear}
    boneToModify {boneName=dyn_shoulderpad_l}
    constraintSetup [
     {targetBone {boneName=upperarm_twist_02_l}}
     {targetBone {boneName=spine_03}}
     {targetBone {boneName=upperarm_l}}
     {targetBone {boneName=clavicle_l}}
     {targetBone {boneName=upperarm_twist_01_l}}
    ]
    constraintWeights [1,1,1,1,1]
   }
   n8 = Get bIsParachuteOpened()
   n9 = Get ParachuteOpened()
   n10 = Get IsSkydiving()
   n11 = "Transform (Modify) Bone - Bone: dyn_shoulderpad_l"(
    ComponentPose=@n7
    Alpha=SelectFloat(A=SelectFloat(A=@n9, B=@n10, bPickA=@n8), B=Get ArmSwingDelta_L() -> (ArmSwingDelta_L:double), bPickA=Get FloatOrDive())
    Translation="0,0,-3"
    Rotation="13,10,-25"
    Scale="1,1,1"
   ) {
    rotationMode=BMM_Additive
    rotationSpace=BCS_BoneSpace
    translationMode=BMM_Additive
    translationSpace=BCS_BoneSpace
    alphaBoolBlend {blendOption=Linear}
    boneToModify {boneName=dyn_shoulderpad_l}
    rotation=<rot 13 10 -25>
    translation {z=-3}
   }
   n12 = "Transform (Modify) Bone - Bone: dyn_shoulderpad_l"!(ComponentPose=@n11, bAlphaBoolEnabled:bool=Get bIsSkydiveFloating(), Translation="2,0,5", Rotation="0,-5,0", Scale="1,1,1") -> (Pose) {
    alphaInputType=Bool
    rotationMode=BMM_Additive
    rotationSpace=BCS_BoneSpace
    translationMode=BMM_Additive
    translationSpace=BCS_BoneSpace
    alphaBoolBlend {blendOption=Linear}
    boneToModify {boneName=dyn_shoulderpad_l}
    rotation {yaw=-5}
    translation {x=2, z=5}
   }
   n13 = "Constraint - const_blade_l (T)"(ComponentPose=@n12, Alpha=1, ConstraintWeights_0=1) {
    alphaBoolBlend {blendOption=Linear}
    boneToModify {boneName=const_blade_l}
    constraintSetup [{transformType=Translation, targetBone {boneName=lowerarm_twist_01_l}}]
    constraintWeights [1]
   }
   n14 = "Constraint - const_blade_l (R)"(ComponentPose=@n13, Alpha=1, ConstraintWeights_0=1) {
    alphaBoolBlend {blendOption=Linear}
    boneToModify {boneName=const_blade_l}
    constraintSetup [{transformType=Rotation, targetBone {boneName=lowerarm_twist_01_l}}]
    constraintWeights [1]
   }
   n15 = "Constraint - const_blade_l (PPPP)"!(
    ComponentPose=@n14
    Alpha=1
    ConstraintWeights_0=0.2838
    ConstraintWeights_1:float=0.2111
    ConstraintWeights_2:float=0.061
    ConstraintWeights_3:float=0.4441
   ) -> (Pose) {
    alphaBoolBlend {blendOption=Linear}
    boneToModify {boneName=const_blade_l}
    constraintSetup [
     {targetBone {boneName=lowerarm_l}}
     {targetBone {boneName=lowerarm_twist_01_l}}
     {targetBone {boneName=upperarm_l}}
     {targetBone {boneName=lowerarm_twist_02_l}}
    ]
    constraintWeights [1,1,1,1]
   }
   n16 = "Constraint - dyn_foot_l (P)"(ComponentPose=@n15, Alpha=1, ConstraintWeights_0=1) {
    alphaBoolBlend {blendOption=Linear}
    boneToModify {boneName=dyn_foot_l}
    constraintSetup [{targetBone {boneName=ball_l}}]
    constraintWeights [1]
   }
   n17 = "Constraint - dyn_foot_r (P)"(ComponentPose=@n16, Alpha=1, ConstraintWeights_0=1) {
    alphaBoolBlend {blendOption=Linear}
    boneToModify {boneName=dyn_foot_r}
    constraintSetup [{targetBone {boneName=ball_r}}]
    constraintWeights [1]
   }
   n18 = Get legSwingDelta_r() -> (legSwingDelta_r:double)
   n19 = "Transform (Modify) Bone - Bone: dyn_sub_thigh_bottom_r"(ComponentPose=@n17, Alpha=@n18, Translation="0,0,0", Rotation="0,0,0", Scale="10,25,10") {
    scaleMode=BMM_Additive
    scaleSpace=BCS_BoneSpace
    alphaBoolBlend {blendOption=Linear}
    boneToModify {boneName=dyn_sub_thigh_bottom_r}
    scale=<10 25 10>
   }
   n20 = "Transform (Modify) Bone - Bone: dyn_sub_thigh_top_r"(ComponentPose=@n19, Alpha=@n18, Translation="0,0,0", Rotation="0,0,0", Scale="2,1.2,2") {
    scaleMode=BMM_Additive
    scaleSpace=BCS_BoneSpace
    alphaBoolBlend {blendOption=Linear}
    boneToModify {boneName=dyn_sub_thigh_top_r}
    scale=<2 1.2 2>
   }
   n21 = Get legSwingDelta_l() -> (legSwingDelta_l:double)
   n22 = "Transform (Modify) Bone - Bone: dyn_sub_thigh_bottom_l"(ComponentPose=@n20, Alpha=@n21, Translation="0,0,0", Rotation="0,0,0", Scale="5,14,5") {
    scaleMode=BMM_Additive
    scaleSpace=BCS_BoneSpace
    alphaBoolBlend {blendOption=Linear}
    boneToModify {boneName=dyn_sub_thigh_bottom_l}
    scale=<5 14 5>
   }
   n23 = "Transform (Modify) Bone - Bone: dyn_sub_thigh_top_l"(ComponentPose=@n22, Alpha=@n21, Translation="0,0,0", Rotation="0,0,0", Scale="2,1.2,2") {
    scaleMode=BMM_Additive
    scaleSpace=BCS_BoneSpace
    alphaBoolBlend {blendOption=Linear}
    boneToModify {boneName=dyn_sub_thigh_top_l}
    scale=<2 1.2 2>
   }
   n24 = "Constraint - const_thighbolt_l (PPPPPPP)"!(
    ComponentPose=@n23
    Alpha=1
    ConstraintWeights_0=0.0321
    ConstraintWeights_1:float=0.0099
    ConstraintWeights_2:float=0.0022
    ConstraintWeights_3:float=0.0003
    ConstraintWeights_4:float=0.0098
    ConstraintWeights_5:float=0.5552
    ConstraintWeights_6:float=0.3904
   ) -> (Pose) {
    alphaBoolBlend {blendOption=Linear}
    boneToModify {boneName=const_thighbolt_l}
    constraintSetup [
     {targetBone {boneName=thigh_l}}
     {targetBone {boneName=spine_01}}
     {targetBone {boneName=spine_02}}
     {targetBone {boneName=spine_03}}
     {targetBone {boneName=thigh_l}}
     {targetBone {boneName=pelvis}}
     {targetBone {boneName=thigh_twist_01_l}}
    ]
    constraintWeights [1,1,1,1,1,1,1]
   }
   n25 = "Constraint - const_thighbolt_r (PPPPPP)"!(
    ComponentPose=@n24
    Alpha=1
    ConstraintWeights_0=0.0073
    ConstraintWeights_1:float=0.5166
    ConstraintWeights_2:float=0.001
    ConstraintWeights_3:float=0.0002
    ConstraintWeights_4:float=0.0001
    ConstraintWeights_5:float=0.4749
   ) -> (Pose) {
    alphaBoolBlend {blendOption=Linear}
    boneToModify {boneName=const_thighbolt_r}
    constraintSetup [
     {targetBone {boneName=thigh_r}}
     {targetBone {boneName=thigh_twist_01_r}}
     {targetBone {boneName=spine_01}}
     {targetBone {boneName=spine_02}}
     {targetBone {boneName=spine_03}}
     {targetBone {boneName=pelvis}}
    ]
    constraintWeights [1,1,1,1,1,1]
   }
   Output Pose(Result=Component To Local(ComponentPose=@n25)) {layerGroup=None}
  ]}
 EventGraph {nodes [
   Event BlueprintUpdateAnimation()
   > n1 = Sequence!(execute) -> (then_0, then_1, then_2:Exec)
   n2 = GetOwnerSkeletalMeshComponent()
   > Set legSwingDelta_r(legSwingDelta_r:double=MapRangeClamped(Value=GetTwistAndSwingAngleOfDeltaRotationFromRefPose(self=@n2, BoneName=thigh_r).OutSwingAngle, InRangeA=148, InRangeB=75)) -> (Output_Get:double)
   > Set legSwingDelta_l(legSwingDelta_l:double=MapRangeClamped(Value=GetTwistAndSwingAngleOfDeltaRotationFromRefPose(self=@n2, BoneName=thigh_l).OutSwingAngle, InRangeA=29, InRangeB=107)) -> (Output_Get:double)
   n3 = BreakRotator(InRot=BreakTransform(InTransform=GetDeltaTransformFromRefPose(self=@n2, BoneName=upperarm_r, BaseName=clavicle_r)).Rotation)
   Set armSwingDelta_r(execute=@n1.then_1, armSwingDelta_r:double=MapRangeClamped(Value="Absolute (Float)"(A=@n3.Yaw), InRangeA=45, InRangeB=116)) -> (Output_Get:double)
   n4 = BreakRotator(InRot=BreakTransform(InTransform=GetDeltaTransformFromRefPose(self=@n2, BoneName=upperarm_l, BaseName=clavicle_l)).Rotation)
   > Set armSwingDelta_l(armSwingDelta_l:double=MapRangeClamped(Value="Absolute (Float)"(A=@n4.Yaw), InRangeA=5, InRangeB=70)) -> (Output_Get:double)
   > Set ArmSwingDeltaINV_R(ArmSwingDeltaINV_R:double=float - float(B=Get ArmSwingDelta_R()) {is=title K2Node_CallFunction}) -> (Output_Get:double)
   n5 = OR Boolean(A=Get bIsSkydiveFloating(), B=Get bIsSkydiveDiveMode())
   Set IsSkydiving(execute=@n1.then_2, IsSkydiving:double=SelectFloat(A=1, B=0, bPickA=@n5)) -> (Output_Get:double)
   > Set ParachuteOpened(ParachuteOpened:double=SelectFloat(A=1, B=0, bPickA=Get bIsParachuteOpened())) -> (Output_Get:double)
   > Set FloatOrDive(FloatOrDive:bool=@n5) -> (Output_Get:bool)
   TryGetPawnOwner()
  ]}
]
variables [
 ArmSwingDelta_L {type=number}
 ArmSwingDelta_R {type=number}
 ArmSwingDeltaINV_R {type=number}
 FloatOrDive {type=boolean}
 IsSkydiving {type=number}
 legSwingDelta_l {type=number}
 legSwingDelta_r {type=number}
 ParachuteOpened {type=number}
]
